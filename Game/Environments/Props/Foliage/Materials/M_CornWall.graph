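# Masked two-sided foliage material for corn wall props, with a Nanite override to MI_CornPatch.
# Vertex red blends dust, roughness and normals and weights panned-noise wind, added to a decaying hit wobble around HitLocation.
# A camera sphere mask cuts opacity nearby, and forward/mobile paths drop wind and detail shading.
# A static LOD switch selects height-gradient shading or this path with destructible damage, both feeding the Creative move-tool function.
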Material /Game/Environments/Props/Foliage/Materials/M_CornWall
graph {
 properties {
  bAutomaticallySetUsageInEditor=false
  bTangentSpaceNormal=false
  bUseMaterialAttributes=true
  bUsedWithNanite=true
  blendMode=BLEND_Masked
  ditheredLODTransition=true
  pixelDepthOffsetMode=PDOM_Legacy
  shadingModel=MSM_TwoSidedFoliage
  twoSided=true
  usedShadingModels=Two Sided Foliage
  naniteOverrideMaterial {
   bEnableOverride=true
   overrideMaterial=null
   overrideMaterialEditor=/Game/Environments/Asteria/Foliage/Plants/Corn/Materials/MI_CornPatch
   overrideMaterialRef=null
  }
 }
 nodes [
  n1 = MaterialFunctionCall(
   Input=MaterialFunctionCall() {materialFunction=&BoundingBoxBased_0-1_UVW}.B
   Input Low=ScalarParameter() {defaultValue=0.2, parameterName=SpecBoundsBottom}
   Input High=ScalarParameter() {defaultValue=0.6, parameterName=SpecBoundsTop}
  ) {materialFunction=&RemapValueRange}
  n2 = Power(Base=Saturate(None=Saturate(None=@n1.Result)), Exp=ScalarParameter() {defaultValue=1.333, parameterName=LODSpecpower})
  n3 = Multiply(
   A=VectorParameter() {parameterName=ColorMult, defaultValue=<rgba 0.3448 0.3448 0.3448 1>}.RGB
   B=TextureSampleParameter2D() {parameterName=D, texture=/Game/Environments/Props/Foliage/Textures/T_Cornwall_D}.RGB
  )
  n4 = Multiply(
   A=LinearInterpolate(A=Constant3Vector() {constant=<rgba 0.32 0.2502 0.0677 0>}, B=Constant3Vector() {constant=<rgba 0.7292 0.7292 0.7292 0>}, Alpha=@n2)
   B=@n3
  )
  n5 = ScalarParameter() {defaultValue=0.125, parameterName=LODSpecTop}
  n6 = SetMaterialAttributes!(
   MaterialAttributes
   Base Color=@n4
   Specular=LinearInterpolate(A=Divide(A=@n5), B=@n5, Alpha=@n2) {constB=0.02}
   Roughness=Constant() {r=1}
   Normal=Constant3Vector() {constant=<rgba 0 0 1 0>}
  ) {attributeSetTypes [
    #1
    #2
    #3
    #4
   ]}
  n7 = VertexColor()
  n8 = LinearInterpolate(Alpha=@n7.R) {constA=0.4, constB=0.7}
  n9 = TextureSampleParameter2D() {parameterName=D, texture=&T_Cornwall_D}
  n10 = LinearInterpolate(Alpha=@n7.R) {constA=0.7, constB=2}
  n11 = Constant() {r=0.125}
  n12 = TextureSampleParameter2D() {parameterName=S, samplerType=SAMPLERTYPE_Masks, texture=/Game/Environments/Props/Foliage/Textures/T_CornWall_S}
  n13 = Multiply(A=@n11, B=@n12.B)
  n14 = Constant() {r=0.95}
  n15 = Constant() {r=0.7}
  n16 = Constant3Vector() {constant=<rgba 1 1 0 0>}
  n17 = Multiply(A=WorldPosition().XYZ, B=@n16)
  n18 = Multiply(A=@n16, B=VectorParameter() {parameterName=HitLocation, defaultValue=<rgba 0 0 0 1>}.RGB)
  n19 = Saturate(None=Divide(A=Distance(A=@n17, B=@n18), B=ScalarParameter() {defaultValue=200, parameterName=WobbleDistance}))
  n20 = Time()
  n21 = ScalarParameter() {defaultValue=3, parameterName=WobbleSpeed}
  n22 = Multiply(A=Sine(None=Multiply(A=@n20, B=@n21)), B=VectorParameter() {parameterName=WobbleDirection, defaultValue=<rgba 1 0 0 1>}.RGB)
  n23 = ScalarParameter() {defaultValue=30, parameterName=WobbleStrength}
  n24 = Divide(A=Subtract(A=Time(), B=ScalarParameter() {parameterName=Time}), B=ScalarParameter() {defaultValue=1.5, parameterName=WobbleSeconds})
  n25 = Multiply(A=Multiply(A=OneMinus(None=@n19), B=Multiply(A=@n22, B=@n23)), B=OneMinus(None=Saturate(None=@n24)))
  n26 = Divide(A=ComponentMask(None=WorldPosition().XYZ) {g=true, r=true}, B=ScalarParameter() {defaultValue=1200, group=WPO, parameterName=WindWaveScale})
  n27 = Multiply(A=Time() {period=1}, B=ScalarParameter() {defaultValue=0.07, group=WPO, parameterName=WindSpeed})
  n28 = Multiply(
   A=TextureSample(UVs=Panner(Coordinate=@n26, Time=@n27) {bFractionalPart=true, speedY=1}) {samplerType=SAMPLERTYPE_Normal, texture=/Game/Global/Textures/Default/Noise/T_SwirlNoise_N}.RGB
   B=ScalarParameter() {defaultValue=12, group=WPO, parameterName=WindPower}
  )
  n29 = LinearInterpolate(A=Constant3Vector() {constant=<rgba 0.0052 0.0035 0 1>}, B=Constant3Vector() {constant=<rgba 0.0327 0.0417 0.0028 1>}, Alpha=@n7.R) {constA=0.15}
  n30 = CameraPositionWS()
  n31 = WorldPosition()
  n32 = CollectionParameter() {collection=&FortniteMaterialParameters, parameterName=GrassPushDownRadius}
  n33 = TextureSampleParameter2D() {parameterName=MaskTexture, samplerType=SAMPLERTYPE_Masks, texture=&T_CornWall_S}
  n34 = Multiply(
   A=Multiply(A=Clamp(None=@n32), B=ScalarParameter() {defaultValue=300, parameterName=FadeRadius})
   B=LinearInterpolate(
    A=ScalarParameter() {defaultValue=0.8, parameterName=FadeMod01}
    B=ScalarParameter() {defaultValue=1.4, parameterName=FadeMod01_1}
    Alpha=StaticSwitchParameter(True=@n9.R, False=@n33.B) {parameterName=UseDiffuseForFadeMask}
   ) {constA=0.8, constB=1.5}
  )
  n35 = ScalarParameter() {defaultValue=0.9, parameterName=FadeHardness}
  n36 = Multiply(A=OneMinus(None=SphereMask(A=@n30, B=@n31.XYZ, Radius=@n34, Hardness=@n35)), B=@n33.R)
  n37 = Constant3Vector() {constant=<rgba 0 0 1 0>}
  n38 = Add(
   A=Multiply(
    A=TextureSampleParameter2D() {parameterName=N, texture=/Game/Environments/Props/Foliage/Textures/T_Cornwall_N}.RGB
    B=Constant3Vector() {constant=<rgba 1 1 0 0>}
   )
   B=Transform(None=@n37) {transformSourceType=TRANSFORMSOURCE_Local, transformType=TRANSFORM_Tangent}
  )
  n39 = LinearInterpolate(Alpha=@n7.R) {constB=1.25}
  n40 = MakeMaterialAttributes!(
   BaseColor=LinearInterpolate(
    A=VectorParameter() {parameterName=DustColor, defaultValue=<rgba 0.2135 0.1294 0.0472 0>}.RGB
    B=Multiply(A=@n8, B=@n9.RGB)
    Alpha=Saturate(None=@n10)
   )
   Metallic
   Specular=@n13
   Roughness=LinearInterpolate(B=LinearInterpolate(A=@n14, B=@n15, Alpha=@n12.B), Alpha=@n7.R) {constA=0.95}
   Anisotropy
   EmissiveColor
   Opacity=ScalarParameter() {defaultValue=1, parameterName=Opacity}
   OpacityMask=@n36
   Normal=LinearInterpolate(A=Transform(None=@n38), B=@n37, Alpha=Saturate(None=@n39))
   Tangent
   WorldPositionOffset=Add(A=@n25, B=Multiply(A=@n7.R, B=@n28))
   SubsurfaceColor=Multiply(A=Multiply(A=@n9.RGB, B=@n29), B=ScalarParameter() {defaultValue=25, parameterName=SSSVal})
   ClearCoat
   ClearCoatRoughness
   AmbientOcclusion
   Refraction
   CustomizedUVs_0
   CustomizedUVs_1
   CustomizedUVs_2
   CustomizedUVs_3
   CustomizedUVs_4
   CustomizedUVs_5
   CustomizedUVs_6
   CustomizedUVs_7
   PixelDepthOffset
   ShadingModel
   Displacement
  )
  n41 = MakeMaterialAttributes(
   BaseColor=Multiply(A=@n9.RGB, B=@n7.R)
   Specular=@n13
   Roughness=Constant() {r=1}
   OpacityMask=@n36
   Normal=Constant3Vector() {constant=<rgba 0 0 1 0>}
   WorldPositionOffset=@n25
  )
  n42 = ShadingPathSwitch(Default=@n40, Deferred=@n40, Forward=@n41, Mobile=@n41)
  n43 = StaticBool()
  n44 = StaticBool() {value=true}
  n45 = VectorParameter() {group=z_code, parameterName=WSDeformationCenter}
  n46 = ScalarParameter() {defaultValue=1, group=z_code, parameterName=WSDamageRadius}
  MaterialFunctionCall(MA In=StaticSwitchParameter(
    True=@n6
    False=MaterialFunctionCall(Material_In=@n42, UseCrackedDamageTextures=@n43, UseAdditiveWPO=@n44, WSDeformationCenter=@n45.RGB, WSDamageRadius=@n46) {materialFunction=&MF_FN_DestructibleDamage}.Material_Out
   ) {parameterName=LOD}) {materialFunction=&MF_CreativeMoveTool}
  NamedRerouteDeclaration(None=@n11) {name=SpecMult, nodeColor=<rgba 1 0 0.8437 1>}
  Reroute(None=ComponentMask(None=VertexNormalWS()) {b=true})
  ShadingPathSwitch(Default=@n4, Deferred=@n4, Forward=@n3, Mobile=@n3)
 ]
}
